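# Unlit material for the Esports02 Pro cosmetic wheel, producing a glowing infinity-style effect on a skeletal mesh.
# MF_Infinity_OffsetFunction layers a dot mask and a gradient ring into the infinity pattern.
# A Center Warp layer distorts radial coordinates with rotating, panning pyramid noise to sample a gradient, weighted by distance from UV center.
# Both are summed, brightened and tinted by Custom Color as emissive, then passed through vehicle damage and de-rez functions.
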Material /VehicleCosmetics/Wheels/Esports02_Pro/Materials/M_Wheel_Esports02_Pro_INF
graph {
 properties {
  bAutomaticallySetUsageInEditor=false
  bUseMaterialAttributes=true
  bUsedWithMeshDeformer=true
  bUsedWithSkeletalMesh=true
  previewMesh=/VehicleCosmetics/Wheels/Esports01/Meshes/SK_Wheel_ESports01
  shadingModel=MSM_Unlit
  usedShadingModels=Unlit
 }
 nodes [
  n1 = NamedRerouteUsage()
  MaterialFunctionCall(In=MaterialFunctionCall(Material=SetMaterialAttributes(Emissive Color=@n1) {attributeSetTypes [#1]}) {materialFunction=&MF_BaseVehicle_Damage}.Result) {materialFunction=&MF_VehicleDeRez_Transition}
  n2 = Rotator(Coordinate=TextureCoordinate() {uTiling=0.3, vTiling=0.3}) {speed=-0.1}
  n3 = ConstantBiasScale(None=TextureSample(UVs=Panner(Coordinate=@n2) {speedX=0.05, speedY=0.5}) {samplerSource=SSM_Wrap_WorldGroupSettings, texture=/VehicleCosmetics/Wheels/_SharedTextures/FX/Pyramids_Skin_D}.R) {bias=-2, scale=1}
  n4 = TextureSample() {
   samplerSource=SSM_Wrap_WorldGroupSettings
   samplerType=SAMPLERTYPE_LinearColor
   texture=/VehicleCosmetics/Wheels/_SharedTextures/FX/Radial_Coords01_D
  }
  n5 = Multiply(A=Panner(Coordinate=Multiply(A=@n3, B=@n4.G)) {speedY=-0.05}, B=Constant2Vector() {g=5, r=1})
  n6 = Add(A=TextureCoordinate()) {constB=-0.5}
  n7 = DotProduct(A=@n6, B=@n6)
  n8 = Multiply(
   A=Power(Base=TextureSample(UVs=@n5) {samplerSource=SSM_Wrap_WorldGroupSettings, texture=/VehicleCosmetics/Wheels/_SharedTextures/FX/Vert_Linear01_D}.RGB) {constExponent=5}
   B=@n7
  )
  NamedRerouteDeclaration(None=Multiply(A=Multiply(A=@n8, B=@n7)) {constB=0.35}) {name=Center Warp, nodeColor=<rgba 0.1016 1 0 1>}
  n9 = TextureObjectParameter() {group=01 INF, parameterName=Infinity Mask, sortPriority=0, texture=/VehicleCosmetics/Wheels/ClockGear_SE/Textures/DotMask_01}
  n10 = TextureCoordinate() {uTiling=5, vTiling=5}
  n11 = Clamp(None=MaterialFunctionCall!(
    Texture=@n9
    Offset=ScalarParameter() {defaultValue=-3, group=01 INF, parameterName=Offset, sortPriority=0}
    UVs=Add(A=@n10) {constB=-2}
    RotateUVs=StaticBool() {value=true}
    GradientFalloff
    GradientDarkness=ScalarParameter() {defaultValue=0.001, group=01 INF, parameterName=Gradient Darkness, sortPriority=1}
   ) {materialFunction=/VehicleCosmetics/Wheels/ClockGear_SE/Materials/MF_Infinity_OffsetFunction}.Result)
  n12 = TextureObjectParameter() {group=01 INF, parameterName=Gradient Ring, sortPriority=0, texture=/VehicleCosmetics/Wheels/_SharedTextures/FX/GradientRing_T}
  n13 = Multiply(A=TextureCoordinate(), B=ScalarParameter() {defaultValue=1, group=01 INF, parameterName=Tile, sortPriority=0})
  n14 = ScalarParameter() {group=01 INF, parameterName=Shift, sortPriority=0}
  n15 = Clamp(None=MaterialFunctionCall(
    Texture=@n12
    Offset=ScalarParameter() {defaultValue=-3, group=01 INF, parameterName=Offset, sortPriority=0}
    UVs=Add(A=@n13, B=@n14) {constB=-2}
    RotateUVs=StaticBool()
    GradientDarkness=ScalarParameter() {defaultValue=0.001, group=01 INF, parameterName=Gradient Darkness, sortPriority=1}
   ) {materialFunction=&MF_Infinity_OffsetFunction}.Result)
  NamedRerouteDeclaration(None=ComponentMask(None=Add(A=@n11, B=@n15)) {r=true}) {name=INF, nodeColor=<rgba 1 0 0.4219 1>}
  n16 = Add(A=NamedRerouteUsage(), B=NamedRerouteUsage())
  NamedRerouteDeclaration(None=Multiply(A=Multiply(A=@n16) {constB=15}, B=VectorParameter() {group=01 INF, parameterName=Custom Color, defaultValue=<rgba 1 0.6308 0.1946 1>}.RGB)) {name=Emissive, nodeColor=<rgba 0 0.7422 1 1>}
 ]
}
