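# Masked LOD material for the Esports02 Pro wheel rim on skeletal meshes, with custom-color glow and an infinity-mask effect.
# Emissive blends two Fresnel-based looks by a UV center mask, using panning noise and an offset-function infinity mask.
# Sparkles multiply reflection-vector noise with UV noise.
# Texture masks, normal map and alpha texture supply surface inputs, and the attributes pass through shared vehicle damage and de-rez transition functions.
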Material /VehicleCosmetics/Wheels/Esports02_Pro/Materials/M_Wheel_Esports02_Pro_LODs
graph {
 properties {
  bAutomaticallySetUsageInEditor=false
  bUseMaterialAttributes=true
  bUsedWithMeshDeformer=true
  bUsedWithSkeletalMesh=true
  blendMode=BLEND_Masked
  previewMesh=/VehicleCosmetics/Wheels/Esports01/Meshes/SK_Wheel_ESports01
  usedShadingModels=Default Lit
 }
 nodes [
  n1 = NamedRerouteUsage()
  n2 = NamedRerouteUsage()
  n3 = NamedRerouteUsage()
  n4 = SetMaterialAttributes!(
   MaterialAttributes
   Base Color=LinearInterpolate(A=Add(A=@n1, B=@n2), Alpha=NamedRerouteUsage()) {constB=0}
   Emissive Color=Add(A=@n2, B=NamedRerouteUsage())
   Normal=NamedRerouteUsage()
   Metallic=LinearInterpolate(A=ScalarParameter() {defaultValue=1, group=00 Base, parameterName=Metalic}, Alpha=@n3) {constB=0}
   Specular=LinearInterpolate(Alpha=@n3) {constA=0.5, constB=0}
   Roughness=LinearInterpolate(A=ScalarParameter() {defaultValue=0.2, group=00 Base, parameterName=Roughness}, Alpha=@n3)
   Opacity Mask=LinearInterpolate(
    A=ScalarParameter() {parameterName=Mask}
    B=Constant() {r=1}
    Alpha=TextureSample() {samplerType=SAMPLERTYPE_LinearColor, texture=/VehicleCosmetics/Wheels/Esports02_Pro/Textures/T_ESports02Pro_Rim_Alpha_LOD}.R
   )
  ) {attributeSetTypes [
    #1
    #2
    #3
    #4
    #5
    #6
    #7
   ]}
  MaterialFunctionCall(In=MaterialFunctionCall(Material=@n4) {materialFunction=&MF_BaseVehicle_Damage}.Result) {materialFunction=&MF_VehicleDeRez_Transition}
  n5 = TextureSampleParameter2D() {
   group=00 Base
   parameterName=Normal
   samplerSource=SSM_Wrap_WorldGroupSettings
   samplerType=SAMPLERTYPE_Normal
   sortPriority=2
   texture=/VehicleCosmetics/Wheels/Esports02_Pro/Textures/T_ESports02Pro_Rim_N_LOD
  }
  NamedRerouteDeclaration(None=@n5.RGB) {name=Normal, nodeColor=<rgba 0.5 0.5 1 1>}
  n6 = TextureSampleParameter2D() {
   group=00 Base
   parameterName=Masks
   samplerSource=SSM_Wrap_WorldGroupSettings
   samplerType=SAMPLERTYPE_LinearColor
   sortPriority=1
   texture=/VehicleCosmetics/Wheels/Esports02_Pro/Textures/T_ESports02Pro_Rim_RGB_LOD
  }
  n7 = TextureObjectParameter() {
   group=01 INF
   parameterName=Infinity Mask
   samplerType=SAMPLERTYPE_LinearColor
   sortPriority=0
   texture=/VehicleCosmetics/Wheels/Esports02_Pro/Textures/Wheel_Emerald_SE_TreadMask
  }
  n8 = Multiply(A=TextureCoordinate(), B=ScalarParameter() {defaultValue=1, group=01 INF, parameterName=Tile})
  n9 = ScalarParameter() {group=01 INF, parameterName=Shift}
  n10 = Clamp(None=MaterialFunctionCall!(
    Texture=@n7
    Offset=ScalarParameter() {defaultValue=0.35, group=01 INF, parameterName=Offset, sortPriority=0}
    UVs=Add(A=@n8, B=@n9)
    RotateUVs=StaticBool()
    GradientFalloff
    GradientDarkness=ScalarParameter() {defaultValue=0.001, group=01 INF, parameterName=Gradient Darkness, sortPriority=1}
   ) {materialFunction=/VehicleCosmetics/Wheels/ClockGear_SE/Materials/MF_Infinity_OffsetFunction}.Result)
  NamedRerouteDeclaration(None=Add(A=@n6.R, B=Multiply(A=@n10) {constB=24})) {name=INF 24, nodeColor=<rgba 0.3125 1 0 1>}
  n11 = VectorParameter() {group=00 Base, parameterName=Custom Color, sortPriority=0, defaultValue=<rgba 0.8963 0.5333 0.1384 1>}
  n12 = NamedRerouteDeclaration(None=AppendVector(A=@n11.RGB, B=@n11.RGB)) {name=Custom Color, nodeColor=<rgba 1 0.5 0 1>}
  n13 = ScalarParameter() {defaultValue=2, group=00 Base, parameterName=Glow Amount Max, sortPriority=3}
  n14 = Panner(Coordinate=TextureCoordinate() {vTiling=0.5}) {speedX=0.1, speedY=-0.1}
  n15 = Panner(Coordinate=TextureCoordinate() {vTiling=0.5}) {speedX=-0.1, speedY=-0.11}
  n16 = Multiply(A=Multiply(
    A=TextureSample(UVs=@n14) {samplerSource=SSM_Wrap_WorldGroupSettings, texture=/VehicleCosmetics/Wheels/_SharedTextures/FX/ContrastNoiseGradient_T}.RGB
    B=TextureSample(UVs=@n15) {samplerSource=SSM_Wrap_WorldGroupSettings, texture=&ContrastNoiseGradient_T}.RGB
   )) {constB=4}
  n17 = MaterialFunctionCall!(Normal=NamedRerouteUsage(), Exponent=ScalarParameter() {defaultValue=1, group=00 Base, parameterName=Fresnel Amount, sortPriority=4}) {materialFunction=/VehicleCosmetics/Wheels/Cola_Holographic/Materials/MF_Fresnel_Classic}
  n18 = Multiply(
   A=LinearInterpolate(A=Multiply(A=@n12) {constB=0.5}, B=Multiply(A=@n13, B=@n12), Alpha=Add(A=@n16, B=@n17.Result)) {desc=Mask for Emis}
   B=NamedRerouteUsage()
  ) {desc=Add INF}
  n19 = NamedRerouteUsage()
  n20 = Constant() {r=3}
  n21 = Multiply(
   A=Multiply(A=MaterialFunctionCall(Normal=@n19, Exponent=@n20) {materialFunction=&MF_Fresnel_Classic}.Result) {constB=2}
   B=NamedRerouteUsage()
  ) {constB=2}
  n22 = ComponentMask(None=TextureCoordinate() {uTiling=1.23}) {r=true}
  n23 = NamedRerouteDeclaration(None=Ceil(None=ConstantBiasScale(None=@n22) {bias=-0.5, scale=1})) {name=Mask for center, nodeColor=<rgba 0 0.9531 1 1>}
  n24 = Add(A=AppendVector(A=@n16, B=Constant() {r=1}), B=NamedRerouteUsage())
  n25 = NamedRerouteUsage()
  NamedRerouteDeclaration(None=Multiply(A=LinearInterpolate(A=@n18, B=@n21, Alpha=@n23), B=Multiply(A=@n24, B=@n25))) {name=Emissive Final, nodeColor=<rgba 1 0.3516 0 1>}
  NamedRerouteDeclaration(None=Add(A=@n6.R, B=@n10)) {name=INF, nodeColor=<rgba 0.3125 1 0 1>}
  NamedRerouteDeclaration(None=@n6.G) {name=Masks_G, nodeColor=<rgba 0 1 0 1>}
  n26 = ComponentMask(None=Add(A=ReflectionVectorWS(), B=NamedRerouteUsage())) {g=true, r=true}
  n27 = Multiply(
   A=TextureSample(UVs=Multiply(A=@n26) {constB=0.0625}) {samplerSource=SSM_Wrap_WorldGroupSettings, texture=/VehicleCosmetics/Wheels/_SharedTextures/FX/Noise_Generic_01_Pack}.B
   B=TextureSample(UVs=TextureCoordinate() {uTiling=4, vTiling=4}) {samplerSource=SSM_Wrap_WorldGroupSettings, texture=&Noise_Generic_01_Pack}.B
  )
  NamedRerouteDeclaration(None=Multiply(A=Multiply(A=@n27) {constB=4096}, B=NamedRerouteUsage())) {name=Sparkles, nodeColor=<rgba 0.2422 0 1 1>}
  NamedRerouteDeclaration(None=@n6.R) {name=Masks_R, nodeColor=<rgba 0 0.4375 1 1>}
  VertexColor()
 ]
}
